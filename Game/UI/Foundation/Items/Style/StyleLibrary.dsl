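; StyleLibrary is a UI function library that supplies shared colors, themes and sizes for item displays and difficulty indicators.
; Stylesheet getters break the Fort UI Stylesheet struct, selecting colors by enum and sizes from brush-size maps, which ResizeBrush uses to resize Slate brushes.
; Conning functions derive difficulty colors from a fixed index, account-level cutoffs evaluated from curve table rows, or the gap to a Homebase Rating data table value.
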(fn GetBase&BuffColors
  (params (Display%20Type :type EFortStatValueDisplayType%20Enum :default "BasicPaired") (Buff%20State :type EFortBuffState%20Enum :default "NoChange") (Stylesheet :type Fort%20UIStylesheet%20Structure))
  (bind (_theme_default_91_015cb18f413b4a467b38cb93e8734a68 _theme_emphasis_94_962f11a944e9df11f547fbbdb98dd4be _theme_critical_97_43f63d0c4e71395016ae00b41bb7ba0c _color_buff_145_ac063fb242d9ecbd0b6306bbe205c7f7 _color_debuff_146_0402338944fcc3ded4c90e9a78fd8bc5 _color_unique_150_b15cf0c9436988001f70b4b8f9d32e5b _color_fireelemental_152_fbafda0c445a96f1b689e880a707069c _color_iceelemental_154_fda209d8429df768aba53eba24c4b505 _color_boltelemental_156_10aefec147c25b2dee47619e4ab4abc7) (Utilities|Struct|BreakFortUIStylesheet Stylesheet))
  (return (Utilities|Select :Index Display%20Type :BasicPaired "(R=1.000000,G=1.000000,B=1.000000,A=1.000000)" :BasicSingle "(R=1.000000,G=1.000000,B=1.000000,A=1.000000)" :Unique _color_unique_150_b15cf0c9436988001f70b4b8f9d32e5b :ElementalFire _color_fireelemental_152_fbafda0c445a96f1b689e880a707069c :ElementalIce _color_iceelemental_154_fda209d8429df768aba53eba24c4b505 :ElementalElectric _color_boltelemental_156_10aefec147c25b2dee47619e4ab4abc7) (Utilities|Select :Index Buff%20State :NoChange "(R=0.000000,G=0.000000,B=0.000000,A=0.000000)" :Better _color_buff_145_ac063fb242d9ecbd0b6306bbe205c7f7 :Worse _color_debuff_146_0402338944fcc3ded4c90e9a78fd8bc5)))

(fn GetBoltElementalColor
  (params (Stylesheet :type Fort%20UIStylesheet%20Structure))
  (return (Utilities|Struct|BreakFortUIStylesheet Stylesheet)))

(fn GetBuffColor
  (params (Stylesheet :type Fort%20UIStylesheet%20Structure))
  (return (Utilities|Struct|BreakFortUIStylesheet Stylesheet)))

(fn GetConningColorForSpecificDifficulty
  (params (Difficulty%20Value :type Integer))
  (switch int Difficulty%20Value
    (:Default
      (return 0 true))
    (:0
      ;[1 1 1] VeryEasy
      (LocalVariables|SetDifficultyColor "(R=0.550000,G=0.550000,B=0.550000,A=1.000000)"))
    (:1
      ;[1 1 1] Easy
      (LocalVariables|SetDifficultyColor "(R=0.123603,G=1.000000,B=0.000000,A=1.000000)"))
    (:2
      ;[1 1 1] Normal
      (LocalVariables|SetDifficultyColor "(R=1.000000,G=0.987922,B=0.000000,A=1.000000)"))
    (:3
      ;[1 1 1] Hard
      (LocalVariables|SetDifficultyColor "(R=1.000000,G=0.213740,B=0.000000,A=1.000000)"))
    (:4
      ;[1 1 1] VeryHard
      (LocalVariables|SetDifficultyColor "(R=0.390000,G=0.000000,B=0.000000,A=1.000000)")))
  (return (LocalVariables|GetDifficultyColor) false))

(fn GetConningColorFromDifficultyValue
  (params (Player%20Account%20Level :type Integer) (Content%20Difficulty%20Level :type Float%20%28double-precision%29))
  (LocalVariables|SetPlayer%20Account%20Level%20In Player%20Account%20Level)
  (LocalVariables|SetContent%20DL%20In Content%20Difficulty%20Level)
  ;[1 1 1] Default outputs to error value
  (LocalVariables|SetDifficultyOut -1)
  (LocalVariables|SetDifficultyColor "(R=0.549020,G=0.160784,B=0.521569,A=1.000000)")
  (LocalVariables|SetDifficultyColorName "NSLOCTEXT(\"[7DA285CF43A30ABF43F861B7E0E51403]\", \"78E5748F4149D2FBECD743937F638EE7\", \"ERROR\")")
  (LocalVariables|SetConningTooltipText "NSLOCTEXT(\"[7DA285CF43A30ABF43F861B7E0E51403]\", \"A3EC9EC44FCA1381BAFBCA85B8A12FCD\", \"ERROR\")")
  (LocalVariables|SetValid true)
  (Utilities|FlowControl|Sequence
    (:then_0
      ;[1 1 1] Store all SP cutoff points for the row
      (for _index (range 16)
        (if (LocalVariables|GetValid)
          (bind _outxy (DataTable|EvaluateCurveTableRow "/Game/Balance/DataTables/AccountLevelConningColor.AccountLevelConningColor" (Utilities|Array|Get(acopy) (LocalVariables|GetRowNames) _index) (LocalVariables|GetContent%20DL%20In) "Style Library Conning")
            (:RowFound
              (Utilities|Array|SetArrayElem (LocalVariables|GetSPCutoffValues) _index _outxy))
            (:RowNotFound
              (LocalVariables|SetValid false))))))
    (:then_1
      ;[1 1 1] Find the first value greater than the player's skill points
      (bind (_array_element _array_index) (Utilities|Array|ForEachLoopwithBreak (LocalVariables|GetSPCutoffValues))
        (:LoopBody
          (if (> _array_element (Math|Conversions|ToFloat(Integer) (LocalVariables|GetPlayer%20Account%20Level%20In)))
            ;[1 1 1] Set return values based on current index
            (bind _output_get (LocalVariables|SetDifficultyOut (Utilities|Array|Get(acopy) (LocalVariables|GetDifficultyIndices) _array_index)))
            (LocalVariables|SetDifficultyColor (Utilities|Array|Get(acopy) (LocalVariables|GetDifficultyColors) _output_get))
            (LocalVariables|SetConningTooltipText (Utilities|Array|Get(acopy) (LocalVariables|GetDifficultyTooltips) _output_get))
            (LocalVariables|SetDifficultyColorName (Utilities|Array|Get(acopy) (LocalVariables|GetDifficultyColorNames) _output_get))))))
    (:then_2
      (return (LocalVariables|GetDifficultyColor) (LocalVariables|GetValid) (LocalVariables|GetConningTooltipText) (LocalVariables|GetDifficultyOut)))))

(fn GetConningColorFromHomebaseRating
  (params (HomebaseRating :type Integer) (DifficultyLevel :type Integer))
  (bind _returnvalue (Utilities|GetDataTableRow "/Game/Balance/DataTables/HomebaseRatingDifficultyMapping.HomebaseRatingDifficultyMapping" (Utilities|String|StringToName (Utilities|String|ToString(Integer) HomebaseRating)))
    (:then
      (if (Math|Integer|InRange(Integer) (- DifficultyLevel (Utilities|Struct|BreakHomebaseRatingDifficultyMappingData _returnvalue)) -1 0)
        (return (Utilities|Array|Get(acopy) (LocalVariables|GetDifficultyColors) 1))
        (elif (Math|Integer|InRange(Integer) (- DifficultyLevel (Utilities|Struct|BreakHomebaseRatingDifficultyMappingData _returnvalue)) -99 -2)
          (return (Utilities|Array|Get(acopy) (LocalVariables|GetDifficultyColors)))
          (elif (Math|Integer|InRange(Integer) (- DifficultyLevel (Utilities|Struct|BreakHomebaseRatingDifficultyMappingData _returnvalue)) 1 1)
            (return (Utilities|Array|Get(acopy) (LocalVariables|GetDifficultyColors) 2))
            (elif (Math|Integer|InRange(Integer) (- DifficultyLevel (Utilities|Struct|BreakHomebaseRatingDifficultyMappingData _returnvalue)) 2 99)
              (return (Utilities|Array|Get(acopy) (LocalVariables|GetDifficultyColors) 3)))))))
    (:RowNotFound
      (return (Utilities|Array|Get(acopy) (LocalVariables|GetDifficultyColors))))))

(fn GetDebuffColor
  (params (Stylesheet :type Fort%20UIStylesheet%20Structure))
  (return (Utilities|Struct|BreakFortUIStylesheet Stylesheet)))

(fn GetFireElementalColor
  (params (Stylesheet :type Fort%20UIStylesheet%20Structure))
  (return (Utilities|Struct|BreakFortUIStylesheet Stylesheet)))

(fn GetHarvestWeakPointColor
  (params (StyleSheet :type Fort%20UIStylesheet%20Structure))
  (bind (_theme_default_91_015cb18f413b4a467b38cb93e8734a68 _theme_emphasis_94_962f11a944e9df11f547fbbdb98dd4be _theme_critical_97_43f63d0c4e71395016ae00b41bb7ba0c _color_buff_145_ac063fb242d9ecbd0b6306bbe205c7f7 _color_debuff_146_0402338944fcc3ded4c90e9a78fd8bc5 _color_unique_150_b15cf0c9436988001f70b4b8f9d32e5b _color_fireelemental_152_fbafda0c445a96f1b689e880a707069c _color_iceelemental_154_fda209d8429df768aba53eba24c4b505 _color_boltelemental_156_10aefec147c25b2dee47619e4ab4abc7 _color_harvestweakpoint_159_ae0cf44c4731daf69193d4b4e59b0e7e) (Utilities|Struct|BreakFortUIStylesheet StyleSheet))
  (return _color_harvestweakpoint_159_ae0cf44c4731daf69193d4b4e59b0e7e))

(fn GetIceElementalColor
  (params (Stylesheet :type Fort%20UIStylesheet%20Structure))
  (return (Utilities|Struct|BreakFortUIStylesheet Stylesheet)))

(fn GetItemIconSize
  (params (Stylesheet :type Fort%20UIStylesheet%20Structure) (Brush%20Size :type EFortBrushSize%20Enum :default "VeryVerySmall"))
  (bind (_color_buff_145_ac063fb242d9ecbd0b6306bbe205c7f7 _color_debuff_146_0402338944fcc3ded4c90e9a78fd8bc5 _color_unique_150_b15cf0c9436988001f70b4b8f9d32e5b _color_fireelemental_152_fbafda0c445a96f1b689e880a707069c _color_iceelemental_154_fda209d8429df768aba53eba24c4b505 _color_boltelemental_156_10aefec147c25b2dee47619e4ab4abc7 _color_harvestweakpoint_159_ae0cf44c4731daf69193d4b4e59b0e7e _standard_icon_size_174_e90b59cc4ef046aedf28a0ba360d7d71 _item_size_173_b43931b34eee416f836301a3a3573e09 _item_icon_size_172_93f895f043e3487e9af205b72c3ef84b) (Utilities|Struct|BreakFortUIStylesheet Stylesheet))
  (return (Utilities|Map|Find _item_icon_size_172_93f895f043e3487e9af205b72c3ef84b Brush%20Size)))

(fn GetItemSize
  (params (Stylesheet :type Fort%20UIStylesheet%20Structure) (Brush%20Size :type EFortBrushSize%20Enum :default "VeryVerySmall"))
  (bind (_color_buff_145_ac063fb242d9ecbd0b6306bbe205c7f7 _color_debuff_146_0402338944fcc3ded4c90e9a78fd8bc5 _color_unique_150_b15cf0c9436988001f70b4b8f9d32e5b _color_fireelemental_152_fbafda0c445a96f1b689e880a707069c _color_iceelemental_154_fda209d8429df768aba53eba24c4b505 _color_boltelemental_156_10aefec147c25b2dee47619e4ab4abc7 _color_harvestweakpoint_159_ae0cf44c4731daf69193d4b4e59b0e7e _standard_icon_size_174_e90b59cc4ef046aedf28a0ba360d7d71 _item_size_173_b43931b34eee416f836301a3a3573e09) (Utilities|Struct|BreakFortUIStylesheet Stylesheet))
  (return (Utilities|Map|Find _item_size_173_b43931b34eee416f836301a3a3573e09 Brush%20Size)))

(fn GetStandardIconSize
  (params (Stylesheet :type Fort%20UIStylesheet%20Structure) (Brush%20Size :type EFortBrushSize%20Enum :default "VeryVerySmall"))
  (bind (_color_buff_145_ac063fb242d9ecbd0b6306bbe205c7f7 _color_debuff_146_0402338944fcc3ded4c90e9a78fd8bc5 _color_unique_150_b15cf0c9436988001f70b4b8f9d32e5b _color_fireelemental_152_fbafda0c445a96f1b689e880a707069c _color_iceelemental_154_fda209d8429df768aba53eba24c4b505 _color_boltelemental_156_10aefec147c25b2dee47619e4ab4abc7 _color_harvestweakpoint_159_ae0cf44c4731daf69193d4b4e59b0e7e _standard_icon_size_174_e90b59cc4ef046aedf28a0ba360d7d71) (Utilities|Struct|BreakFortUIStylesheet Stylesheet))
  (return (Utilities|Map|Find _standard_icon_size_174_e90b59cc4ef046aedf28a0ba360d7d71 Brush%20Size)))

(fn GetTheme
  (params (Stylesheet :type Fort%20UIStylesheet%20Structure) (Theme%20Type :type EFortUITheme%20Enum :default "NewEnumerator0"))
  (bind (_theme_default_91_015cb18f413b4a467b38cb93e8734a68 _theme_emphasis_94_962f11a944e9df11f547fbbdb98dd4be _theme_critical_97_43f63d0c4e71395016ae00b41bb7ba0c) (Utilities|Struct|BreakFortUIStylesheet Stylesheet))
  (return (Utilities|Select :Index Theme%20Type :NewEnumerator0 _theme_default_91_015cb18f413b4a467b38cb93e8734a68 :NewEnumerator9 _theme_emphasis_94_962f11a944e9df11f547fbbdb98dd4be :NewEnumerator10 _theme_critical_97_43f63d0c4e71395016ae00b41bb7ba0c)))

(fn GetThemeColor
  (params (Stylesheet :type Fort%20UIStylesheet%20Structure) (Theme%20Type :type EFortUITheme%20Enum :default "NewEnumerator0") (Color%20Type :type EFortUIThemeColor%20Enum :default "NewEnumerator0"))
  (bind (_light_2_baf9e4af40c0a5c1b86a439b79b3a1e4 _faded_5_2fcc8a1c4ba623597b91e882e0a95f26 _active_8_10ae283b44a455de7af248b79132a69c _normal_11_f26c9bab4ec26a2f2b63a6a5f67ae748 _dark_14_b49822cc48fc28e72dfacca1f36b2c18 _ultra_17_d9b7b92b475b19e3c7733c85fa62631c) (Utilities|Struct|BreakFortUITheme (CallFunction|GetTheme Stylesheet Theme%20Type)))
  (return (Utilities|Select :Index Color%20Type :NewEnumerator0 _light_2_baf9e4af40c0a5c1b86a439b79b3a1e4 :NewEnumerator1 _faded_5_2fcc8a1c4ba623597b91e882e0a95f26 :NewEnumerator2 _active_8_10ae283b44a455de7af248b79132a69c :NewEnumerator3 _normal_11_f26c9bab4ec26a2f2b63a6a5f67ae748 :NewEnumerator4 _dark_14_b49822cc48fc28e72dfacca1f36b2c18 :NewEnumerator5 _ultra_17_d9b7b92b475b19e3c7733c85fa62631c)))

(fn GetUniqueColor
  (params (Stylesheet :type Fort%20UIStylesheet%20Structure))
  (return (Utilities|Struct|BreakFortUIStylesheet Stylesheet)))

(fn ResizeBrush
  (params (BrushSize :type EFortBrushSize%20Enum :default "VeryVerySmall") (BaseBrush :type Slate%20Brush%20Structure))
  (bind _icon_size (CallFunction|GetStandardIconSize :Brush%20Size BrushSize))
  (bind _structout (Utilities|Struct|SetmembersinSlateBrush BaseBrush _icon_size))
  (return _structout))
